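# Material function that computes BASE's animated emissive intensity from a start animation and a death animation, also outputting both timelines.
# Pause-independent time minus each start-time parameter is divided by its play rate (effectively a duration) and clamped, with the death timeline inverted.
# The timelines sample red and green curves in one input lookup texture, multiplied together with a power multiplier.
# A max-float default start time holds the death animation at its initial state.
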MaterialFunction /Figment_S07/LM_7_00/Packages/PBW/BASE/Materials/BASE_Animation
graph {nodes [
  n1 = ScalarParameter() {defaultValue=1, group=BASE, parameterName=BASE_PowerMultiplier}
  n2 = Time() {bIgnorePause=true}
  n3 = ScalarParameter() {group=BASE, parameterName=BASE_AnimStartTime}
  n4 = Clamp(None=Divide(A=Subtract(A=@n2, B=@n3), B=ScalarParameter() {defaultValue=1, group=BASE, parameterName=BASE_AnimPlayRate}))
  n5 = FunctionInput(Preview=TextureObject() {samplerType=SAMPLERTYPE_LinearColor, texture=/Figment_S07/LM_7_00/Packages/PBW/BASE/Textures/AnimatedBASELight}) {bUsePreviewValueAsDefault=true, inputType=FunctionInput_Texture2D}
  n6 = Multiply(A=TextureSample(UVs=FunctionOutput(None=@n4) {outputName=Animation Timeline, sortPriority=-1}, Tex=@n5) {samplerType=SAMPLERTYPE_LinearColor, texture=&AnimatedBASELight}.R) {constB=3}
  n7 = Subtract(A=Time() {bIgnorePause=true}, B=ScalarParameter() {defaultValue=3.4028234663852886e+38, group=BASE, parameterName=BASE_DeathAnimStartTime})
  n8 = ScalarParameter() {defaultValue=0.7, group=BASE, parameterName=BASE_DeathAnimPlayRate}
  n9 = OneMinus(None=Clamp(None=Divide(A=@n7, B=@n8)))
  FunctionOutput(None=Multiply(
    A=Multiply(A=@n1, B=@n6)
    B=TextureSample(UVs=@n9, Tex=@n5) {samplerType=SAMPLERTYPE_LinearColor, texture=&AnimatedBASELight}.G
   )) {outputName=BASE_EmissiveOutput, sortPriority=-1}
  FunctionOutput(None=@n9) {outputName=Animation Timeline1, sortPriority=-1}
 ]}
